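# Sound cue for drawing the BeanDog slingshot, played as weapon foley with weapon-reload attenuation.
# A mixer layers one of three random draw one-shots with a full-tension loop wave.
# The tension layer sits behind a fixed delay and an Enveloper with a rising volume curve, so it follows the initial draw.
# A Modulator at the root applies default pitch and volume variation to the mix.
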SoundCue /WildEstateWeapons/Gameplay/BeanDog/Audio/IP_SC_BeanDog_Slingshot_Draw
root=Modulator_0
nodes [
 {id=Delay_0, children [WavePlayer_0], properties {delayMax=1, delayMin=1}}
 {
  id=Enveloper_0
  children [Delay_0]
  properties {pitchCurve {editorCurveData {keys [{time=0}]}}, volumeCurve {editorCurveData {keys [<at 0 0>,<at 0.5 1>]}}}
 }
 {id=Mixer_0, children [Random_0,Enveloper_0], properties {inputVolume [1,1]}}
 {id=Modulator_0, children [Mixer_0]}
 {id=Random_0, children [WavePlayer_4,WavePlayer_5,WavePlayer_6], properties {weights [1,1,1]}}
 {id=WavePlayer_0, properties {soundWaveAssetPtr=/WildEstateWeapons/Gameplay/BeanDog/Audio/Waves/IP_BeanDog_Slingshot_Tension_Full_Loop_01}}
 {id=WavePlayer_4, properties {soundWaveAssetPtr=/WildEstateWeapons/Gameplay/BeanDog/Audio/Waves/IP_BeanDog_Slingshot_Draw_03}}
 {id=WavePlayer_5, properties {soundWaveAssetPtr=/WildEstateWeapons/Gameplay/BeanDog/Audio/Waves/IP_BeanDog_Slingshot_Draw_02}}
 {id=WavePlayer_6, properties {soundWaveAssetPtr=/WildEstateWeapons/Gameplay/BeanDog/Audio/Waves/IP_BeanDog_Slingshot_Draw_01}}
]
properties {
 attenuationSettings=&Fort_Weapon_Reload_Att
 bDebug=true
 bOverrideConcurrency=true
 maxDistance=2250
 soundClassObject=&Fort_WeaponFoley
 volumeMultiplier=0.6
}
