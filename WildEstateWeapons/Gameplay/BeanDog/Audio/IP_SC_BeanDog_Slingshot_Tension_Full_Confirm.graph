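# Sound cue for the BeanDog slingshot weapon's full-tension confirmation sound.
# A Random node picks one of three equally weighted wave variations.
# It plays as weapon foley at roughly half volume, using the shared weapon reload attenuation.
# The cue overrides concurrency and has its debug flag enabled.
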SoundCue /WildEstateWeapons/Gameplay/BeanDog/Audio/IP_SC_BeanDog_Slingshot_Tension_Full_Confirm
root=Random_0
nodes [
 {id=Random_0, children [WavePlayer_1,WavePlayer_2,WavePlayer_3], properties {weights [1,1,1]}}
 {id=WavePlayer_1, properties {soundWaveAssetPtr=/WildEstateWeapons/Gameplay/BeanDog/Audio/Waves/IP_BeanDog_Slingshot_Tension_Full_Confirm_01}}
 {id=WavePlayer_2, properties {soundWaveAssetPtr=/WildEstateWeapons/Gameplay/BeanDog/Audio/Waves/IP_BeanDog_Slingshot_Tension_Full_Confirm_02}}
 {id=WavePlayer_3, properties {soundWaveAssetPtr=/WildEstateWeapons/Gameplay/BeanDog/Audio/Waves/IP_BeanDog_Slingshot_Tension_Full_Confirm_03}}
]
properties {
 attenuationSettings=&Fort_Weapon_Reload_Att
 bDebug=true
 bOverrideConcurrency=true
 maxDistance=2250
 soundClassObject=&Fort_WeaponFoley
 volumeMultiplier=0.501
}
