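# Sound cue for the close pass-by whoosh of a BeanDog slingshot projectile, in the ranged-weapon projectile effects sound class.
# A Random node picks among four equally weighted compound-bow arrow waves and feeds a root Modulator with pitch fixed at 1.
# Spatial behavior comes from the shared compound-bow whoosh attenuation asset; the cue overrides concurrency and has its debug flag on.
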SoundCue /WildEstateWeapons/Gameplay/BeanDog/Audio/IP_SC_BeanDog_Slingshot_Whoosh_By_Close
root=Modulator_0
nodes [
 {id=Modulator_0, children [Random_0], properties {pitchMax=1, pitchMin=1}}
 {id=Random_0, children [WavePlayer_0,WavePlayer_5,WavePlayer_6,WavePlayer_7], properties {weights [1,1,1,1]}}
 {id=WavePlayer_0, properties {soundWaveAssetPtr=/WildEstateWeapons/Gameplay/BeanDog/Audio/Waves/Weapon_Bow_Compound_Arrow_By_Close_01}}
 {id=WavePlayer_5, properties {soundWaveAssetPtr=/WildEstateWeapons/Gameplay/BeanDog/Audio/Waves/Weapon_Bow_Compound_Arrow_By_Close_02}}
 {id=WavePlayer_6, properties {soundWaveAssetPtr=/WildEstateWeapons/Gameplay/BeanDog/Audio/Waves/Weapon_Bow_Compound_Arrow_By_Close_03}}
 {id=WavePlayer_7, properties {soundWaveAssetPtr=/WildEstateWeapons/Gameplay/BeanDog/Audio/Waves/Weapon_Bow_Compound_Arrow_By_Close_04}}
]
properties {
 attenuationSettings=/Game/Sounds/Fort_Weapons/Explosive_Bow/Weapon_Compound_Bow_Whoosh_By_Att
 bDebug=true
 bOverrideConcurrency=true
 maxDistance=2097152
 soundClassObject=&Fort_WeaponRanged_Projectile_Effects
 volumeMultiplier=1
 attenuationOverrides {
  bAttenuate=false
  bEnableOcclusion=true
  bEnableReverbSend=false
  falloffDistance=0
  occlusionLowPassFilterFrequency=2000
  occlusionTraceChannel=ECC_GameTraceChannel11
  occlusionVolumeAttenuation=0
  spatializationAlgorithm=SPATIALIZATION_HRTF
  attenuationShapeExtents {x=50000}
 }
}
