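# Sound cue for a medium whoosh-by of the BeanDog slingshot projectile in WildEstateWeapons.
# A Random node picks with equal weight among four compound-bow arrow fly-by waves, played at reduced volume.
# It routes to the ranged-weapon projectile effects sound class and references the Explosive Bow whoosh-by attenuation asset.
# Concurrency override and the debug flag are both switched on.
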SoundCue /WildEstateWeapons/Gameplay/BeanDog/Audio/IP_SC_BeanDog_Slingshot_Whoosh_By_Med
root=Random_1
nodes [
 {id=Random_1, children [WavePlayer_1,WavePlayer_2,WavePlayer_3,WavePlayer_4], properties {weights [1,1,1,1]}}
 {id=WavePlayer_1, properties {soundWaveAssetPtr=/WildEstateWeapons/Gameplay/BeanDog/Audio/Waves/Weapon_Bow_Compound_Arrow_By_Med_01}}
 {id=WavePlayer_2, properties {soundWaveAssetPtr=/WildEstateWeapons/Gameplay/BeanDog/Audio/Waves/Weapon_Bow_Compound_Arrow_By_Med_02}}
 {id=WavePlayer_3, properties {soundWaveAssetPtr=/WildEstateWeapons/Gameplay/BeanDog/Audio/Waves/Weapon_Bow_Compound_Arrow_By_Med_03}}
 {id=WavePlayer_4, properties {soundWaveAssetPtr=/WildEstateWeapons/Gameplay/BeanDog/Audio/Waves/Weapon_Bow_Compound_Arrow_By_Med_04}}
]
properties {
 attenuationSettings=/Game/Sounds/Fort_Weapons/Explosive_Bow/Weapon_Compound_Bow_Whoosh_By_Att
 bDebug=true
 bOverrideConcurrency=true
 maxDistance=2097152
 soundClassObject=&Fort_WeaponRanged_Projectile_Effects
 volumeMultiplier=0.4
 attenuationOverrides {
  bAttenuate=false
  bEnableOcclusion=true
  bEnableReverbSend=false
  bUseComplexCollisionForOcclusion=true
  falloffDistance=1000
  occlusionTraceChannel=ECC_GameTraceChannel11
  occlusionVolumeAttenuation=0
  attenuationShapeExtents {x=50000}
 }
}
