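; Weapon Blueprint for BeanDog, a chargeable bow-style WildEstate weapon, handling its cosmetic feedback.
; Equipping attaches an idle Niagara light effect, caches the sound library and binds trigger events; unequipping destroys the effect.
; Pressing the trigger hides the reticle and starts a repeating timer that scales force feedback and camera shake with charge, ending the shake near full; releasing reverses this.
; Fire audio selects first- or third-person and partial or charged sounds.
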(event FortWeapon|EventOnUnEquippedAndDetached
  (bind _idle_fx (Variables|Default|GetIdle_FX)
    (:then
      (Components|DestroyComponent _idle_fx))
    (:else)))

(event FortWeapon|EventOnEquippedAndAttached
  (Networking|CanExecuteCosmeticEvents
    (:True
      (EventDispatchers|BindEventtoOnPressTrigger (EventDispatchers|CreateEvent :!event Custom|StartChargeRumble))
      (EventDispatchers|BindEventtoOnReleaseTrigger (EventDispatchers|CreateEvent :!event Custom|StopRumble))
      (bind _returnvalue (Niagara|SpawnSystemAttached "/WildEstateWeapons/Assets/FX/BeanDog/NS_WildEstate_BeanDog_Light.NS_WildEstate_BeanDog_Light" (Variables|Visual|GetWeaponMesh) "root"))
      (Variables|Default|SetIdle_FX _returnvalue)
      (Rendering|SetRenderCustomDepth (Variables|Default|GetAmmo_Mesh) true)
      (Variables|Default|SetWeapon%20Sound%20Library%20Component (SoundLibrary|GetSoundLibraryComponent "(TagName=\"SoundLibrary.Weapon\")")))))

(event Custom|StartChargeRumble
  (params (Weapon :type Fort%20Weapon%20Object%20Reference))
  (Networking|CanExecuteCosmeticEvents
    (:True
      (Variables|Default|SetIsCharging true)
      (CallFunction|HideReticle)
      (Variables|Default|SetOwningController (Fortnite|GetFortPlayerControllerfromActor (Variables|ParentBow|GetPlayerPawn)))
      (bind _returnvalue (Utilities|Time|SetTimerbyEvent (EventDispatchers|CreateEvent :!event Custom|UpdateFeedback) 0.1 true))
      (Variables|Default|SetChargeRumbleTimer _returnvalue)
      (Variables|Default|GetOwningController
        (:then
          (bind _returnvalue_2 (CameraShakes|StartCameraShake :self (Variables|/Script/Engine.PlayerController|GetPlayerCameraManager (Variables|Default|GetOwningController)) :ShakeClass "/WildEstateWeapons/Gameplay/BeanDog/B_BeanDog_ChargedCameraShake.B_BeanDog_ChargedCameraShake_C" :Scale 0.0 :PlaySpace "CameraLocal"))
          (Variables|Default|SetChargeCameraShake _returnvalue_2)
          (Game|Feedback|PlayDynamicForceFeedback (Variables|Default|GetOwningController) (Math|Float|Clamp(Float) (FortWeapon|GetChargePercent)) -1 true true true true))
        (:else)))))

(event Custom|UpdateFeedback
  (Networking|CanExecuteCosmeticEvents
    (:True
      (Variables|Default|GetOwningController
        (:then
          (CallFunction|UpdateCamShake)
          (Game|Feedback|PlayDynamicForceFeedback (Variables|Default|GetOwningController) (Math|Float|Clamp(Float) (FortWeapon|GetChargePercent)) -1 true true true true))
        (:else)))))

(event Custom|StopRumble
  (params (Weapon :type Fort%20Weapon%20Object%20Reference))
  (Development|PrintString "Stop Rumble!")
  (Utilities|Time|ClearandInvalidateTimerbyHandle (Variables|Default|GetChargeRumbleTimer))
  (Networking|CanExecuteCosmeticEvents
    (:True
      (Variables|Default|SetIsCharging false)
      (CallFunction|ShowReticle)
      (Variables|Default|GetOwningController
        (:then
          (CallFunction|StopCamShake)
          (Game|Feedback|PlayDynamicForceFeedback (Variables|Default|GetOwningController) (Math|Float|Clamp(Float) (FortWeapon|GetChargePercent)) -1 true true true true))
        (:else)))))

(event Custom|UpdateCamShake
  (bind _returnvalue (FortWeapon|GetChargePercent))
  (Networking|CanExecuteCosmeticEvents
    (:True
      (if (> _returnvalue 0.9)
        (CallFunction|StopCamShake)
        (else
          (bind _chargecamerashake (Variables|Default|GetChargeCameraShake)
            (:then
              (Variables|/Script/Engine.CameraShakeBase|SetShakeScale (Math|Float|Clamp(Float) (* _returnvalue 3.0) 0.0 2.0) _chargecamerashake))
            (:else)))))))

(event Custom|StopCamShake
  (CameraShakes|StopCameraShake (Variables|/Script/Engine.PlayerController|GetPlayerCameraManager (Variables|Default|GetOwningController)) (Variables|Default|GetChargeCameraShake) true))

(event EventPlayWeaponFireCustomAudio
  (params (bPersistentFire :type Boolean) (bSecondaryFire :type Boolean))
  ;[0.1875 0.1029 0.06738] Fire Sounds
  (bind _instigator (Variables|Actor|GetInstigator)
    (:then
      (if (Pawn|IsLocallyViewed _instigator)
        (CallFunction|PlayFireSound1p)
        (else
          (CallFunction|PlayFireSound3p))))
    (:else)))

(event Custom|Play%20Fire%20Sound%201p
  ;[0.1875 0.1029 0.06738] Fire Sounds
  (switch Utilities|FlowControl|Switch|SwitchonEFortWeaponChargeStateForFireFX (FortWeapon|GetWeaponChargeStateforFireFX)
    (:Partial
      _)
    (:Full
      (bind _weapon_sound_library_component (Variables|Default|GetWeapon%20Sound%20Library%20Component :!label _label)
        (:then
          (SoundLibrary|PlayEvent _weapon_sound_library_component "(TagName=\"SoundLibrary.Weapon.Fire.Charged.1P\")"))
        (:else)))
    (:Over
      (goto _label))
    (:Max_None
      _))
  (bind _weapon_sound_library_component_2 (Variables|Default|GetWeapon%20Sound%20Library%20Component)
    (:then
      (SoundLibrary|PlayEvent _weapon_sound_library_component_2 "(TagName=\"SoundLibrary.Weapon.Fire.Partial.1P\")"))
    (:else)))

(event Custom|Play%20Fire%20Sound%203p
  ;[0.1875 0.1029 0.06738] Fire Sounds
  (switch Utilities|FlowControl|Switch|SwitchonEFortWeaponChargeStateForFireFX (FortWeapon|GetWeaponChargeStateforFireFX)
    (:Partial
      _)
    (:Full
      (bind _weapon_sound_library_component (Variables|Default|GetWeapon%20Sound%20Library%20Component :!label _label)
        (:then
          (SoundLibrary|PlayEvent _weapon_sound_library_component "(TagName=\"SoundLibrary.Weapon.Fire.Charged.3P\")"))
        (:else)))
    (:Over
      (goto _label))
    (:Max_None
      _))
  (bind _weapon_sound_library_component_2 (Variables|Default|GetWeapon%20Sound%20Library%20Component)
    (:then
      (SoundLibrary|PlayEvent _weapon_sound_library_component_2 "(TagName=\"SoundLibrary.Weapon.Fire.Partial.3P\")"))
    (:else)))

(fn TryShowReticle
  (Utilities|Casting|CastToFortPlayerControllerGameplay (Game|GetPlayerController 0)
    (:then
      (if (Variables|Default|GetIsCharging)
        (CallFunction|HideReticle :!label _label)))
    (:CastFailed
      (if (not (or (Scope|CanUseScopedCameraViaOptic) (Scope|UseScopeTargeting)))
        (goto _label))))
  (CallFunction|ShowReticle))

(fn ConstructionScript
  (|CallParentFunction))
