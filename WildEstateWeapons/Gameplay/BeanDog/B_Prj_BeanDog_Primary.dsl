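; Projectile Blueprint for the BeanDog weapon's primary fire, a thin child that leaves most behaviour to its parent class.
; On the hide-projectile-on-impact event, it sets the "Impact" bool on the trail's Niagara component so the effect can react.
; GetInstigatorPlayerController returns the owning pawn variable, and the construction script only calls the parent.
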(event EventHideProjectileOnImpact
  (Niagara|SetNiagaraVariable(Bool) (Variables|Default|GetProjectileTrail_Niagara) "Impact" true))

(fn GetInstigatorPlayerController
  (return (Variables|Default|GetOwningPawn)))

(fn ConstructionScript
  (|CallParentFunction))
