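; Parent BeanDog rod projectile that explodes on impact and spawns a sludge field, with a timed fallback explosion for long bouncing flights.
; Clients animate a fake tracer through material parameters.
; Water overlaps explode with splash FX and destroy the actor; solid hits set replicated hit data on authority.
; OnRep_HitLocation spawns explosion FX, sludge field, decal (skipped on mobile and player hits), camera cue and sound, then delays trail shutdown.
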(event EventBeginPlay
  (bind _tracermesh (Variables|Default|GetTracerMesh))
  ;[1 1 1] EVENT -> BeginPlay
  ;;[1 1 1] With bouncy surfaces, there's a possibility for a long life. Explode after a time
  (Utilities|Time|SetTimerbyEvent (EventDispatchers|CreateEvent :!event DoExplosion) 22.0)
  ;[1 1 1] Animate fake tracer mesh
  (Networking|CanExecuteCosmeticEvents
    (:True
      (bind _asfort_player_pawn (Utilities|Casting|CastToFortPlayerPawn (Variables|Actor|GetInstigator)))
      (Rendering|Material|SetVectorParameterValueonMaterials _tracermesh "BulletOrigin" (Effects|GetMuzzleLocation (Variables|/Script/FortniteGame.FortPawn|GetCurrentWeapon _asfort_player_pawn) -1))
      (Rendering|Material|SetScalarParameterValueonMaterials _tracermesh "StartTime" (Utilities|Time|GetGameTimeinSeconds))
      (Rendering|Material|SetScalarParameterValueonMaterials _tracermesh "WPODuration" (Math|Float|MapRangeClamped (Math|Vector|VectorLength (Transformation|GetVelocity)) 5e+03 4e+04 0.1 0.25)))))

(event EventOnTouched
  (params (OtherActor :type Actor%20Object%20Reference) (OtherComp :type Primitive%20Component%20Object%20Reference) (HitResult :type Hit%20Result%20Structure%20%28by%20ref%29) (bIsOverlap :type Boolean))
  (bind (_bblockinghit _binitialoverlap _time _distance _location _impactpoint _normal _impactnormal _physmat _hitactor) (Collision|BreakHitResult HitResult))
  (bind _waterinteractioncomponent (Variables|Water|GetWaterInteractionComponent))
  (bind (_waterplanelocation _waterplanenormal) (FX|GetLastWaterSurfaceInfo _waterinteractioncomponent :!class /Script/FortniteGame.FortWaterInteractionComponent))
  (bind _returnvalue (Math|Conversions|RotationFromXVector _waterplanenormal))
  (bind _returnvalue_2 (Math|Rotator|MakeRotfromX _normal))
  (bind (_asfort_pawn _bsuccess) (Utilities|Casting|CastToFortPawn _hitactor))
  ;[1 1 1] EVENT -> OnTouched
  (if bIsOverlap
    (Utilities|IsValid _physmat
      (:Is%20Valid
        (if (CallFunction|HitWater _hitactor)
          (Delivery|DoExplosionatLocation _location)
          (if (Variables|Default|GetUse%20Niagara%20Water%20Impact)
            (Effects|Components|ParticleSystem|SpawnFXAtLocation (Variables|Default|GetWaterImpactFX_NS) _location _returnvalue)
            (else
              (Effects|SpawnEmitteratLocation (Variables|Default|GetWaterImpactFX) _location _returnvalue)))
          (Building|SpawnBuildingGameplayActor "/WildEstateWeapons/Gameplay/BeanDog/BGA_WildEstate_BeanDog_Sludge_Field_Athena.BGA_WildEstate_BeanDog_Sludge_Field_Athena_C" (Math|Transform|MakeTransform _location _returnvalue_2) (Variables|Actor|GetInstigator))
          (Actor|DestroyActor))))
    (else
      (Networking|SwitchHasAuthority
        (:Authority
          (if (not (Delivery|IsBeingKilled))
            (Variables|BPrjBoomBow|SetHitRotation _returnvalue_2)
            (Variables|BPrjBoomBow|SetHitPlayer _bsuccess)
            (Variables|BPrjBoomBow|SetHitLocation _location :!pinned true) ; All Explosion FX
          ))))))

(event OnEnterWaterEvent(WaterInteractionComponent)
  (params (WaterBody :type Fort%20Water%20Body%20Actor%20Object%20Reference) (WaterInteractionComponent :type Fort%20Water%20Interaction%20Component%20Object%20Reference) (bIsFirstBody :type Boolean))
  (Audio|PlaySoundatLocation "/Game/Athena/Sounds/Weapons/Explosions/Expl_Debris_Water_Cue.Expl_Debris_Water_Cue" (Transformation|GetActorLocation)))

(event Custom|DestroyTrail
  (Components|Activation|Deactivate (Variables|Default|GetProjectileTrail_Niagara)))

(event Custom|HideProjectileOnImpact)

(fn HitWater
  (params (OverlapActor :type Actor%20Object%20Reference))
  (Utilities|Casting|CastToBGA_Athena_FlopperSpawn_Parent OverlapActor
    (:then
      _)
    (:CastFailed
      (Utilities|Casting|CastToFortWaterBodyActor OverlapActor
        (:then
          _)
        (:CastFailed
          (return false)))))
  (return true))

(fn OnRep_HitLocation
  (bind _hitrotation (Variables|BPrjBoomBow|GetHitRotation))
  (bind _hitlocation (Variables|BPrjBoomBow|GetHitLocation))
  (bind _hitlocation_2 (Variables|BPrjBoomBow|GetHitLocation))
  (bind _hitrotation_2 (Variables|BPrjBoomBow|GetHitRotation))
  (bind _hitlocation_3 (Variables|BPrjBoomBow|GetHitLocation))
  (if (Variables|BPrjBoomBow|GetUse%20Niagara%20Explosion)
    (bind _returnvalue (Effects|Components|ParticleSystem|SpawnFXAtLocation (Variables|BPrjBoomBow|GetNS_Explosion) _hitlocation_2 _hitrotation_2))
    (Effects|Components|ParticleSystem|SetFloatParameter _returnvalue "Global Scale" 0.63)
    (else
      (bind _returnvalue_2 (Effects|SpawnEmitteratLocation (Variables|BPrjBoomBow|GetP_Explosion) _hitlocation_2 _hitrotation_2))))
  (Rendering|SetRenderCustomDepth _returnvalue true)
  (LOD|SetMaxDrawDistance _returnvalue_2 (Variables|Default|GetDrawDistance))
  (Utilities|FlowControl|Sequence
    (:then_0
      (GameplayCue|ExecuteGameplayCue%28Burst%29Local self (Variables|Default|GetFeedbackCue) (Ability|GameplayCue|MakeGameplayCueParameters 0.0)) ; Cam shake + controller vibration
      (Building|SpawnBuildingGameplayActor "/WildEstateWeapons/Gameplay/BeanDog/BGA_WildEstate_BeanDog_Sludge_Field_Athena.BGA_WildEstate_BeanDog_Sludge_Field_Athena_C" (Math|Transform|MakeTransform _hitlocation _hitrotation) (Variables|Actor|GetInstigator))
      (if (not (or (Fortnite|IsMobilePlatform) (Variables|BPrjBoomBow|GetHitPlayer)))
        (bind _returnvalue_3 (Rendering|Decal|SpawnDecalatLocation "/WildEstateWeapons/Assets/FX/BeanDog/MD_WildEstate_BeanDog_Decal_Impact.MD_WildEstate_BeanDog_Decal_Impact" (Variables|Default|GetDecalSize) _hitlocation (Math|Rotator|CombineRotators _hitrotation "0.000000,0.000000,180.000000") 4.5))
        (Rendering|Components|Decal|SetFadeOut _returnvalue_3 0.1 5.0 false)))
    (:then_1
      (Audio|PlaySoundatLocation (Variables|Sounds|GetCue_CloseSound) _hitlocation_3) ; Play Close Explosion
      (Fortnite|ShowSoundIndicatorLocalforTeam :Location _hitlocation_3 :MaxAudibleDistance (Variables|/Script/Engine.SoundBase|GetMaxDistance (Variables|Sounds|GetCue_DistantSound)) :IndicatorType "Gunshot" :Team (Team|GetActorTeam self) :Affiliations (Utilities|Array|MakeArray "Hostile" "Neutral") :Instigator self :Tint "(R=0.000000,G=0.000000,B=0.000000,A=0.000000)" :AttenuationMaxRangeMultiplier 1.0 :AttenuationIntensityMultiplier 1.0) ; Visual Sound Indicator
      (CallFunction|HideProjectileonImpact)
      (bind _returnvalue_4 (Utilities|Time|SetTimerbyEvent (EventDispatchers|CreateEvent :!event DestroyTrail) 1.0))
      (Variables|Default|SetTrailDelayTimer _returnvalue_4))))

(fn ConstructionScript
  (LOD|SetMaxDrawDistance (Variables|Default|GetTracerMesh) (Variables|Default|GetDrawDistance)))
