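; Gameplay ability for the BeanDog weapon's alternate fire, which shoots a stink arrow that leaves a stink cloud.
; It plays the Explosive Bow fire montage and spawns the Stink Bow arrow from the nearest-pawn aim location along base aim rotation, at a scalable-float speed.
; When the projectile stops, it spawns the stink cloud building gameplay actor there, disables that actor's collision and only then applies the fire cooldown effect to the owner.
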(event Ability|EventActivateAbility
  (bind _asfort_player_pawn (Utilities|Casting|CastToFortPlayerPawn (Ability|GetAvatarActorfromActorInfo)))
  (Ability|Tasks|PlayMontageAndWait "None" "/Game/Animation/Game/MainPlayer/Combat/Guns/ExplosiveBow/ExplosiveBow_Fire_M.ExplosiveBow_Fire_M"
    (:then
      (bind _returnvalue (Utility|GetAimLocationNearestPawn _asfort_player_pawn))
      (bind (_asynctaskproxy _projectiledata) (Projectile|Tasks|SpawnProjectileAndWait "/PrimalGameplay/Items/Bows/Stink/B_Prj_Arrow_Stink_Athena.B_Prj_Arrow_Stink_Athena_C" _asfort_player_pawn _returnvalue (Pawn|GetBaseAimRotation _asfort_player_pawn) (Pawn|GetBaseAimRotation _asfort_player_pawn) (ScalableFloat|GetValueatLevel (Variables|Default|GetRow_Speed)))
        (:Stopped
          (bind _returnvalue_2 (Building|SpawnBuildingGameplayActorfromWorldContext (Variables|Default|GetStinkCloudBGAClass) (Math|Conversions|ToTransform(Vector) (Transformation|GetActorLocation (Utilities|Struct|BreakProjectileEventData _projectiledata)))))
          (Ability|ApplyGameplayEffectToOwner (Variables|Default|GetGE_FireCooldown))
          (Collision|SetActorEnableCollision _returnvalue_2 false))))
    (:OnCompleted
      _)
    (:OnBlendOut
      _)
    (:OnInterrupted
      _)
    (:OnCancelled
      _))
  (Ability|EndAbility))
